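# Sound cue that plays when the Nitro state is applied, on the gadgets sound class.
# A Random node picks one of six equally weighted wave variants, excluded from branch culling, and feeds a Modulator at the root.
# Spatialization uses a player-apply attenuation asset, with overall volume reduced.
# Overridden concurrency allows one instance per owner and blocks new plays while one is active.
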SoundCue /NitroGameplay/Assets/Audio/Cues/SC_NitroState_Applied
root=Modulator_0
nodes [
 {id=Modulator_0, children [Random_0]}
 {
  id=Random_0
  children [WavePlayer_0,WavePlayer_1,WavePlayer_2,WavePlayer_3,WavePlayer_4,WavePlayer_5]
  properties {bShouldExcludeFromBranchCulling=true, weights [1,1,1,1,1,1]}
 }
 {id=WavePlayer_0, properties {soundWaveAssetPtr=/NitroGameplay/Assets/Audio/Waves/NitroState_Applied_01}}
 {id=WavePlayer_1, properties {soundWaveAssetPtr=/NitroGameplay/Assets/Audio/Waves/NitroState_Applied_02}}
 {id=WavePlayer_2, properties {soundWaveAssetPtr=/NitroGameplay/Assets/Audio/Waves/NitroState_Applied_03}}
 {id=WavePlayer_3, properties {soundWaveAssetPtr=/NitroGameplay/Assets/Audio/Waves/NitroState_Applied_04}}
 {id=WavePlayer_4, properties {soundWaveAssetPtr=/NitroGameplay/Assets/Audio/Waves/NitroState_Applied_05}}
 {id=WavePlayer_5, properties {soundWaveAssetPtr=/NitroGameplay/Assets/Audio/Waves/NitroState_Applied_06}}
]
properties {
 attenuationSettings=/NitroGameplay/Assets/Audio/Settings/Att_NitroState_Player_Apply
 bOverrideConcurrency=true
 maxDistance=8300
 soundClassObject=&Fort_Gadgets
 volumeMultiplier=0.5
 concurrencyOverrides {bLimitToOwner=true, maxCount=1, resolutionRule=PreventNew}
}
