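# Looping sound cue for the applied Nitro state, limited to one instance per owner.
# An IsLocal branch plays a random full loop when true, with an envelope that swells quickly, then drops in volume and pitch.
# False or unset waits two seconds, then plays a random band-passed loop at low modulated volume with a slow fade-in.
# Both paths share the same third-person attenuation asset.
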SoundCue /NitroGameplay/Assets/Audio/Cues/SC_NitroState_Applied_Loop
root=Branch_0
nodes [
 {
  id=Attenuation_0
  children [Modulator_0]
  properties {attenuationSettings=/NitroGameplay/Assets/Audio/Settings/Att_NitroState_Player_Bashable_Loop_3P}
 }
 {
  id=Attenuation_1
  children [Enveloper_0]
  properties {attenuationSettings=&Att_NitroState_Player_Bashable_Loop_3P}
 }
 {id=Branch_0, children [Attenuation_1, Delay_0, Delay_0], properties {boolParameterName=IsLocal}}
 {id=Delay_0, children [Attenuation_0], properties {delayMax=2, delayMin=2}}
 {
  id=Enveloper_0
  children [Random_0]
  properties {
   pitchCurve {editorCurveData {keys [<at 0 1.25>, {time=6}]}}
   volumeCurve {editorCurveData {keys [<at 0 0>, {arriveTangent=0.0333, interpMode=RCIM_Cubic, leaveTangent=0.0333, time=0.5, value=1}, <at 6 0.2>]}}
  }
 }
 {
  id=Enveloper_1
  children [Random_1]
  properties {pitchCurve {editorCurveData {keys [{time=0}]}}, volumeCurve {editorCurveData {keys [<at 0 0>,<at 3.1956 0.6375>,<at 7 1>]}}}
 }
 {id=Modulator_0, children [Enveloper_1], properties {volumeMax=0.25, volumeMin=0.2}}
 {id=Random_0, children [WavePlayer_0,WavePlayer_1,WavePlayer_2], properties {bShouldExcludeFromBranchCulling=true, weights [1,1,1]}}
 {id=Random_1, children [WavePlayer_3,WavePlayer_4,WavePlayer_5], properties {bShouldExcludeFromBranchCulling=true, weights [1,1,1]}}
 {id=WavePlayer_0, properties {bLooping=true, soundWaveAssetPtr=/NitroGameplay/Assets/Audio/Waves/NitroState_Applied_Loop_03}}
 {id=WavePlayer_1, properties {bLooping=true, soundWaveAssetPtr=/NitroGameplay/Assets/Audio/Waves/NitroState_Applied_Loop_02}}
 {id=WavePlayer_2, properties {bLooping=true, soundWaveAssetPtr=/NitroGameplay/Assets/Audio/Waves/NitroState_Applied_Loop_01}}
 {id=WavePlayer_3, properties {bLooping=true, soundWaveAssetPtr=/NitroGameplay/Assets/Audio/Waves/NitroState_Applied_BandPassed_Loop_01}}
 {id=WavePlayer_4, properties {bLooping=true, soundWaveAssetPtr=/NitroGameplay/Assets/Audio/Waves/NitroState_Applied_BandPassed_Loop_02}}
 {id=WavePlayer_5, properties {bLooping=true, soundWaveAssetPtr=/NitroGameplay/Assets/Audio/Waves/NitroState_Applied_BandPassed_Loop_03}}
]
properties {
 bOverrideConcurrency=true
 maxDistance=3500
 soundClassObject=&Fort_Gadgets
 volumeMultiplier=0.8
 concurrencyOverrides {bLimitToOwner=true, maxCount=1, resolutionRule=PreventNew}
}
