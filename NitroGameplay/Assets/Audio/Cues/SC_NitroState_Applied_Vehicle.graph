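# Sound cue for the one-shot played when the Nitro state is applied to a vehicle.
# A Random node picks one of six equally weighted apply waves and feeds it through a Modulator at default settings.
# It uses its own vehicle apply attenuation and the gadgets sound class, at reduced volume with virtualization disabled.
# A concurrency override limits simultaneous instances per owner.
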SoundCue /NitroGameplay/Assets/Audio/Cues/SC_NitroState_Applied_Vehicle
root=Modulator_0
nodes [
 {id=Modulator_0, children [Random_1]}
 {
  id=Random_1
  children [WavePlayer_6, WavePlayer_7, WavePlayer_8, WavePlayer_9, WavePlayer_10, WavePlayer_11]
  properties {weights [1,1,1,1,1,1]}
 }
 {id=WavePlayer_10, properties {soundWaveAssetPtr=/NitroGameplay/Assets/Audio/Waves/Nitro_State_Apply_Vehicle_05}}
 {id=WavePlayer_11, properties {soundWaveAssetPtr=/NitroGameplay/Assets/Audio/Waves/Nitro_State_Apply_Vehicle_06}}
 {id=WavePlayer_6, properties {soundWaveAssetPtr=/NitroGameplay/Assets/Audio/Waves/Nitro_State_Apply_Vehicle_01}}
 {id=WavePlayer_7, properties {soundWaveAssetPtr=/NitroGameplay/Assets/Audio/Waves/Nitro_State_Apply_Vehicle_02}}
 {id=WavePlayer_8, properties {soundWaveAssetPtr=/NitroGameplay/Assets/Audio/Waves/Nitro_State_Apply_Vehicle_03}}
 {id=WavePlayer_9, properties {soundWaveAssetPtr=/NitroGameplay/Assets/Audio/Waves/Nitro_State_Apply_Vehicle_04}}
]
properties {
 attenuationSettings=/NitroGameplay/Assets/Audio/Settings/Att_NitroState_Veh_Apply
 bOverrideConcurrency=true
 maxDistance=8300
 soundClassObject=&Fort_Gadgets
 virtualizationMode=Disabled
 volumeMultiplier=0.3
 concurrencyOverrides {bLimitToOwner=true, maxCount=3}
}
