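# Sound cue for the impact of bashing a pawn during the Nitro state, judging by its asset names.
# A Random node picks one of six equally weighted wave variations, and a Modulator at the root varies its playback.
# It uses the player bash impact attenuation and the Fort_Gadgets sound class.
# A concurrency override tightly caps simultaneous instances, and virtualization is disabled.
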SoundCue /NitroGameplay/Assets/Audio/Cues/SC_NitroState_Bash_Pawn
root=Modulator_0
nodes [
 {id=Modulator_0, children [Random_1]}
 {
  id=Random_1
  children [WavePlayer_6, WavePlayer_7, WavePlayer_8, WavePlayer_9, WavePlayer_10, WavePlayer_11]
  properties {bShouldExcludeFromBranchCulling=true, weights [1,1,1,1,1,1]}
 }
 {id=WavePlayer_10, properties {soundWaveAssetPtr=/NitroGameplay/Assets/Audio/Waves/Nitro_State_Bash_Pawn_02}}
 {id=WavePlayer_11, properties {soundWaveAssetPtr=/NitroGameplay/Assets/Audio/Waves/Nitro_State_Bash_Pawn_01}}
 {id=WavePlayer_6, properties {soundWaveAssetPtr=/NitroGameplay/Assets/Audio/Waves/Nitro_State_Bash_Pawn_06}}
 {id=WavePlayer_7, properties {soundWaveAssetPtr=/NitroGameplay/Assets/Audio/Waves/Nitro_State_Bash_Pawn_05}}
 {id=WavePlayer_8, properties {soundWaveAssetPtr=/NitroGameplay/Assets/Audio/Waves/Nitro_State_Bash_Pawn_04}}
 {id=WavePlayer_9, properties {soundWaveAssetPtr=/NitroGameplay/Assets/Audio/Waves/Nitro_State_Bash_Pawn_03}}
]
properties {
 attenuationSettings=/NitroGameplay/Assets/Audio/Settings/Att_NitroState_Player_Bash_Impact_Att
 bOverrideConcurrency=true
 maxDistance=10000
 soundClassObject=&Fort_Gadgets
 virtualizationMode=Disabled
 volumeMultiplier=0.8
 concurrencyOverrides {maxCount=2, volumeScaleAttackTime=0.1}
}
