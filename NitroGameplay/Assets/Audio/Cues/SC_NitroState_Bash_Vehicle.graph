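# Sound cue for the nitro-state vehicle bash, going by its name and its player bash impact attenuation asset.
# A Random node picks among six equally weighted wave variations and feeds a Modulator at the root.
# The cue plays in the gadgets sound class at slightly reduced volume, with virtualization disabled and a concurrency override capping it at two simultaneous instances.
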SoundCue /NitroGameplay/Assets/Audio/Cues/SC_NitroState_Bash_Vehicle
root=Modulator_0
nodes [
 {id=Modulator_0, children [Random_0]}
 {
  id=Random_0
  children [WavePlayer_0,WavePlayer_1,WavePlayer_2,WavePlayer_3,WavePlayer_4,WavePlayer_5]
  properties {bShouldExcludeFromBranchCulling=true, weights [1,1,1,1,1,1]}
 }
 {id=WavePlayer_0, properties {soundWaveAssetPtr=/NitroGameplay/Assets/Audio/Waves/Nitro_State_Bash_Vehicle_01}}
 {id=WavePlayer_1, properties {soundWaveAssetPtr=/NitroGameplay/Assets/Audio/Waves/Nitro_State_Bash_Vehicle_02}}
 {id=WavePlayer_2, properties {soundWaveAssetPtr=/NitroGameplay/Assets/Audio/Waves/Nitro_State_Bash_Vehicle_03}}
 {id=WavePlayer_3, properties {soundWaveAssetPtr=/NitroGameplay/Assets/Audio/Waves/Nitro_State_Bash_Vehicle_04}}
 {id=WavePlayer_4, properties {soundWaveAssetPtr=/NitroGameplay/Assets/Audio/Waves/Nitro_State_Bash_Vehicle_05}}
 {id=WavePlayer_5, properties {soundWaveAssetPtr=/NitroGameplay/Assets/Audio/Waves/Nitro_State_Bash_Vehicle_06}}
]
properties {
 attenuationSettings=/NitroGameplay/Assets/Audio/Settings/Att_NitroState_Player_Bash_Impact_Att
 bOverrideConcurrency=true
 maxDistance=10000
 soundClassObject=&Fort_Gadgets
 virtualizationMode=Disabled
 volumeMultiplier=0.8
 concurrencyOverrides {maxCount=2, volumeScaleAttackTime=0.1}
}
