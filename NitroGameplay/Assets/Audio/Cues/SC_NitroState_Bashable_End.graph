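# Sound cue that plays when the Nitro "Bashable" state ends, as a gadget-class sound.
# A Random node picks one of four equally weighted end-sound waves and feeds it through a Modulator for variation.
# It plays at reduced volume with the Nitro player-apply attenuation.
# A concurrency override allows one instance per owner and blocks new plays while one is active.
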SoundCue /NitroGameplay/Assets/Audio/Cues/SC_NitroState_Bashable_End
root=Modulator_0
nodes [
 {id=Modulator_0, children [Random_0]}
 {
  id=Random_0
  children [WavePlayer_0,WavePlayer_1,WavePlayer_2,WavePlayer_3]
  properties {bShouldExcludeFromBranchCulling=true, weights [1,1,1,1]}
 }
 {id=WavePlayer_0, properties {soundWaveAssetPtr=/NitroGameplay/Assets/Audio/Waves/Nitro_State_Bashable_End_04}}
 {id=WavePlayer_1, properties {soundWaveAssetPtr=/NitroGameplay/Assets/Audio/Waves/Nitro_State_Bashable_End_03}}
 {id=WavePlayer_2, properties {soundWaveAssetPtr=/NitroGameplay/Assets/Audio/Waves/Nitro_State_Bashable_End_02}}
 {id=WavePlayer_3, properties {soundWaveAssetPtr=/NitroGameplay/Assets/Audio/Waves/Nitro_State_Bashable_End_01}}
]
properties {
 attenuationSettings=/NitroGameplay/Assets/Audio/Settings/Att_NitroState_Player_Apply
 bOverrideConcurrency=true
 maxDistance=8300
 soundClassObject=&Fort_Gadgets
 volumeMultiplier=0.3
 concurrencyOverrides {bLimitToOwner=true, maxCount=1, resolutionRule=PreventNew}
}
