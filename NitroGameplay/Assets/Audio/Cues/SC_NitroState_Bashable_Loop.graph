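# Looping gadget-class sound cue for the Nitro bashable state.
# An IsLocal branch plays a dedicated loop for the local player, while false or unset plays one of three random "applied" loops, quieter and with third-person attenuation.
# Each path runs through a modulator and an enveloper that fades in quickly then tapers, more steeply on the remote path.
# Concurrency allows one instance per owner and blocks new starts.
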SoundCue /NitroGameplay/Assets/Audio/Cues/SC_NitroState_Bashable_Loop
root=Branch_0
nodes [
 {
  id=Attenuation_1
  children [Modulator_1]
  properties {attenuationSettings=/NitroGameplay/Assets/Audio/Settings/Att_NitroState_Player_Bashable_Loop_3P}
 }
 {
  id=Attenuation_2
  children [Modulator_0]
  properties {attenuationSettings=/NitroGameplay/Assets/Audio/Settings/Att_NitroState_Player_Bashable_Loop}
 }
 {id=Branch_0, children [Attenuation_2, Attenuation_1, Attenuation_1], properties {boolParameterName=IsLocal}}
 {
  id=Enveloper_0
  children [WavePlayer_0]
  properties {pitchCurve {editorCurveData {keys [{time=0}]}}, volumeCurve {editorCurveData {keys [<at 0 0>,<at 0.2 1>,<at 10 0.8>]}}}
 }
 {
  id=Enveloper_1
  children [Random_0]
  properties {pitchCurve {editorCurveData {keys [{time=0}]}}, volumeCurve {editorCurveData {keys [<at 0 0>,<at 0.2 1>,<at 6 0.4>]}}}
 }
 {id=Modulator_0, children [Enveloper_0]}
 {id=Modulator_1, children [Enveloper_1], properties {volumeMax=0.55, volumeMin=0.5}}
 {id=Random_0, children [WavePlayer_1,WavePlayer_3,WavePlayer_4], properties {bShouldExcludeFromBranchCulling=true, weights [1,1,1]}}
 {id=WavePlayer_0, properties {bLooping=true, soundWaveAssetPtr=/NitroGameplay/Assets/Audio/Waves/Nitro_State_Bashable_Loop}}
 {id=WavePlayer_1, properties {bLooping=true, soundWaveAssetPtr=/NitroGameplay/Assets/Audio/Waves/NitroState_Applied_Loop_01}}
 {id=WavePlayer_3, properties {bLooping=true, soundWaveAssetPtr=/NitroGameplay/Assets/Audio/Waves/NitroState_Applied_Loop_02}}
 {id=WavePlayer_4, properties {bLooping=true, soundWaveAssetPtr=/NitroGameplay/Assets/Audio/Waves/NitroState_Applied_Loop_03}}
]
properties {
 bOverrideConcurrency=true
 maxDistance=8300
 soundClassObject=&Fort_Gadgets
 volumeMultiplier=0.7
 concurrencyOverrides {bLimitToOwner=true, maxCount=1, resolutionRule=PreventNew}
}
