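# Sound cue for the end of the Nitro state's bashable vehicle effect, routed through the gadgets sound class.
# An equal-weight random pick from four end waves feeds a modulator that varies pitch and volume, then an attenuation node using the player bashable-vehicle settings.
# A concurrency override allows one instance per owner and rejects new plays while one is active.
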SoundCue /NitroGameplay/Assets/Audio/Cues/SC_NitroState_Bashable_Veh_End
root=Attenuation_0
nodes [
 {
  id=Attenuation_0
  children [Modulator_0]
  properties {attenuationSettings=/NitroGameplay/Assets/Audio/Settings/Att_NitroState_Player_Bashable_Vehicle}
 }
 {id=Modulator_0, children [Random_1], properties {pitchMin=0.85, volumeMax=1, volumeMin=0.6}}
 {
  id=Random_1
  children [WavePlayer_4,WavePlayer_5,WavePlayer_6,WavePlayer_7]
  properties {bShouldExcludeFromBranchCulling=true, weights [1,1,1,1]}
 }
 {id=WavePlayer_4, properties {soundWaveAssetPtr=/NitroGameplay/Assets/Audio/Waves/Nitro_State_Bashable_Vehicle_End_01}}
 {id=WavePlayer_5, properties {soundWaveAssetPtr=/NitroGameplay/Assets/Audio/Waves/Nitro_State_Bashable_Vehicle_End_02}}
 {id=WavePlayer_6, properties {soundWaveAssetPtr=/NitroGameplay/Assets/Audio/Waves/Nitro_State_Bashable_Vehicle_End_03}}
 {id=WavePlayer_7, properties {soundWaveAssetPtr=/NitroGameplay/Assets/Audio/Waves/Nitro_State_Bashable_Vehicle_End_04}}
]
properties {
 attenuationSettings=&Att_NitroState_Player_Bashable_Vehicle
 bOverrideConcurrency=true
 maxDistance=13850
 soundClassObject=&Fort_Gadgets
 volumeMultiplier=0.6
 concurrencyOverrides {bLimitToOwner=true, maxCount=1, resolutionRule=PreventNew}
}
