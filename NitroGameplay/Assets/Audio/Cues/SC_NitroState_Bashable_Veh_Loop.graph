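# Looping sound cue for the Nitro state's bashable vehicle effect, in the gadgets sound class.
# An IsLocal branch picks one of two chains playing the same looping wave through the same attenuation settings.
# Both fade in with a volume envelope; the non-local or unset chain adds a pitch envelope and slightly raised random pitch.
# Concurrency allows one instance per owner and blocks new ones.
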SoundCue /NitroGameplay/Assets/Audio/Cues/SC_NitroState_Bashable_Veh_Loop
root=Branch_0
nodes [
 {
  id=Attenuation_1
  children [Modulator_1]
  properties {attenuationSettings=/NitroGameplay/Assets/Audio/Settings/Att_NitroState_Player_Bashable_Vehicle_Loop}
 }
 {
  id=Attenuation_2
  children [Modulator_0]
  properties {attenuationSettings=&Att_NitroState_Player_Bashable_Vehicle_Loop}
 }
 {id=Branch_0, children [Attenuation_2, Attenuation_1, Attenuation_1], properties {boolParameterName=IsLocal}}
 {
  id=Enveloper_0
  children [WavePlayer_0]
  properties {
   pitchCurve {editorCurveData {keys [{time=0}]}}
   volumeCurve {editorCurveData {keys [
      <at 0 0 RCIM_Cubic>
      {arriveTangent=0.086, interpMode=RCIM_Cubic, leaveTangent=0.086, tangentMode=RCTM_User, time=1.8238, value=0.9382}
      <at 10 0.8>
     ]}}
  }
 }
 {
  id=Enveloper_2
  children [WavePlayer_2]
  properties {
   pitchCurve {editorCurveData {keys [<at 0 0.8 RCIM_Cubic>, {arriveTangent=0.0167, interpMode=RCIM_Cubic, leaveTangent=0.0167, time=5, value=1.25}, {time=12}]}}
   volumeCurve {editorCurveData {keys [
      <at 0 0 RCIM_Cubic>
      {arriveTangent=0.086, interpMode=RCIM_Cubic, leaveTangent=0.086, tangentMode=RCTM_User, time=1.8238, value=0.9382}
      <at 10 0.8>
     ]}}
  }
 }
 {id=Modulator_0, children [Enveloper_0], properties {pitchMax=1, pitchMin=1, volumeMax=1, volumeMin=1}}
 {id=Modulator_1, children [Enveloper_2], properties {pitchMax=1.1, pitchMin=1.05, volumeMax=1, volumeMin=1}}
 {id=WavePlayer_0, properties {bLooping=true, soundWaveAssetPtr=/NitroGameplay/Assets/Audio/Waves/Nitro_State_Bashable_Vehicle_Loop}}
 {id=WavePlayer_2, properties {bLooping=true, soundWaveAssetPtr=&Nitro_State_Bashable_Vehicle_Loop}}
]
properties {
 bOverrideConcurrency=true
 maxDistance=13850
 soundClassObject=&Fort_Gadgets
 volumeMultiplier=0.4
 concurrencyOverrides {bLimitToOwner=true, maxCount=1, resolutionRule=PreventNew}
}
