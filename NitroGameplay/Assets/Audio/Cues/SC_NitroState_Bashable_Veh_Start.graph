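# Sound cue for the start of the Nitro bashable vehicle state in NitroGameplay, routed to the Fort_Gadgets sound class.
# A Random node picks one of five equally weighted start waves, and a Modulator varies its pitch and volume.
# It uses a dedicated bashable vehicle attenuation asset.
# A concurrency override allows one instance per owner and blocks new plays while one is active.
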SoundCue /NitroGameplay/Assets/Audio/Cues/SC_NitroState_Bashable_Veh_Start
root=Modulator_0
nodes [
 {id=Modulator_0, children [Random_0], properties {pitchMax=1.1, pitchMin=0.85, volumeMax=1, volumeMin=0.6}}
 {
  id=Random_0
  children [WavePlayer_0,WavePlayer_1,WavePlayer_2,WavePlayer_3,WavePlayer_4]
  properties {bShouldExcludeFromBranchCulling=true, weights [1,1,1,1,1]}
 }
 {id=WavePlayer_0, properties {soundWaveAssetPtr=/NitroGameplay/Assets/Audio/Waves/Nitro_State_Bashable_Vehicle_Start_01}}
 {id=WavePlayer_1, properties {soundWaveAssetPtr=/NitroGameplay/Assets/Audio/Waves/Nitro_State_Bashable_Vehicle_Start_02}}
 {id=WavePlayer_2, properties {soundWaveAssetPtr=/NitroGameplay/Assets/Audio/Waves/Nitro_State_Bashable_Vehicle_Start_03}}
 {id=WavePlayer_3, properties {soundWaveAssetPtr=/NitroGameplay/Assets/Audio/Waves/Nitro_State_Bashable_Vehicle_Start_04}}
 {id=WavePlayer_4, properties {soundWaveAssetPtr=/NitroGameplay/Assets/Audio/Waves/Nitro_State_Bashable_Vehicle_Start_05}}
]
properties {
 attenuationSettings=/NitroGameplay/Assets/Audio/Settings/Att_NitroState_Player_Bashable_Vehicle
 bOverrideConcurrency=true
 maxDistance=13850
 soundClassObject=&Fort_Gadgets
 volumeMultiplier=0.95
 concurrencyOverrides {bLimitToOwner=true, maxCount=1, resolutionRule=PreventNew}
}
