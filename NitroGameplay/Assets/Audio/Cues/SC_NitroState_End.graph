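# Sound cue that plays the end sound of the Nitro state, randomly picking one of five equally weighted wave variations.
# An IsLocal Branch feeds all three inputs from the same Modulator and Random node, so every case shares one chain.
# It uses player attenuation and the Gadgets sound class, with virtualization disabled.
# Concurrency allows one instance per owner and blocks new plays while one is active.
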SoundCue /NitroGameplay/Assets/Audio/Cues/SC_NitroState_End
root=Branch_0
nodes [
 {id=Branch_0, children [Modulator_0,Modulator_0,Modulator_0], properties {boolParameterName=IsLocal}}
 {id=Modulator_0, children [Random_1]}
 {
  id=Random_1
  children [WavePlayer_6, WavePlayer_7, WavePlayer_8, WavePlayer_9, WavePlayer_10]
  properties {bShouldExcludeFromBranchCulling=true, weights [1,1,1,1,1]}
 }
 {id=WavePlayer_10, properties {soundWaveAssetPtr=/NitroGameplay/Assets/Audio/Waves/Nitro_State_End_05}}
 {id=WavePlayer_6, properties {soundWaveAssetPtr=/NitroGameplay/Assets/Audio/Waves/Nitro_State_End_01}}
 {id=WavePlayer_7, properties {soundWaveAssetPtr=/NitroGameplay/Assets/Audio/Waves/Nitro_State_End_02}}
 {id=WavePlayer_8, properties {soundWaveAssetPtr=/NitroGameplay/Assets/Audio/Waves/Nitro_State_End_03}}
 {id=WavePlayer_9, properties {soundWaveAssetPtr=/NitroGameplay/Assets/Audio/Waves/Nitro_State_End_04}}
]
properties {
 attenuationSettings=/NitroGameplay/Assets/Audio/Settings/Att_NitroState_Player_Apply
 bOverrideConcurrency=true
 maxDistance=8300
 soundClassObject=&Fort_Gadgets
 virtualizationMode=Disabled
 volumeMultiplier=0.5
 concurrencyOverrides {bLimitToOwner=true, maxCount=1, resolutionRule=PreventNew}
}
